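# UI-domain base material for Parallel-themed Item Shop art, compositing an animated background with two corner decorations.
# An aspect-adjusted series texture receives additive two-color smoke, made from panning honeycomb masked by an exponential radial gradient.
# Upper-left and lower-right corner textures are scaled from widget pixel size and icon parameters, then blended over it using their RGB as alpha.
# Several side branches, including web pattern and distortion, feed nothing in this composite.
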Material /ItemShop/UI/Art/Parallel/M_UI_Parallel_Base
graph {
 properties {bUsedWithStaticMesh=false, materialDomain=MD_UI, pixelDepthOffsetMode=PDOM_Legacy, usedShadingModels=Default Lit}
 nodes [
  Add(A=Multiply(
    A=VectorParameter() {parameterName=Foreground Web Opacity, defaultValue=<rgba 0.5833 0.5833 0.5833 5.6892>}.RGB
    B=TextureSample() {texture=&/T_UI_Parallel_Pattern}.A
   ))
  n1 = TextureCoordinate()
  Add(A=@n1, B=Multiply(B=ScalarParameter() {defaultValue=0.1, group=Smoke, parameterName=Smoke-Distortion-Intensity}))
  n2 = MaterialFunctionCall(Float2=TextureCoordinate() {coordinateIndex=3}) {materialFunction=&BreakOutFloat2Components}
  n3 = Divide(A=@n2.R, B=Max(A=@n2.G))
  n4 = AppendVector(A=@n3, B=Constant() {r=1})
  n5 = TextureCoordinate() {coordinateIndex=4}
  n6 = Multiply(A=OneMinus(None=@n3)) {constB=0.5}
  n7 = Constant()
  n8 = TextureSampleParameter2D(UVs=Add(A=Multiply(A=@n4, B=@n5), B=AppendVector(A=@n6, B=@n7))) {group=Series, parameterName=SeriesTexture, texture=/OfferCatalog/Art/Parallel/T_UI_Parallel_Background}
  n9 = MaterialFunctionCall() {materialFunction=&GetUserInterfaceUV}
  n10 = MaterialFunctionCall(Float2=@n9.Pixel Size) {materialFunction=&BreakOutFloat2Components}
  n11 = Divide(A=@n10.G, B=@n10.G)
  n12 = AppendVector(A=Constant() {r=1}, B=@n11)
  n13 = Constant()
  n14 = Multiply(A=OneMinus(None=@n11)) {constB=0.5}
  n15 = MaterialFunctionCall() {materialFunction=&GetUserInterfaceUV}
  n16 = MaterialFunctionCall(Float2=@n15.Pixel Size) {materialFunction=&BreakOutFloat2Components}
  n17 = Divide(A=@n16.R, B=@n16.R)
  n18 = OneMinus(None=@n17)
  n19 = LinearInterpolate(
   A=Add(A=Multiply(A=@n9.Normalized UV, B=@n12), B=AppendVector(A=@n13, B=@n14))
   B=AppendVector(A=Multiply(A=@n18) {constB=0.5}, B=Constant())
  )
  n20 = TextureSample(UVs=Multiply(A=Panner(Coordinate=@n19) {speedX=0.0296, speedY=0.0333}, B=ScalarParameter() {defaultValue=4.2975, group=Smoke, parameterName=Suit Scale})) {texture=/OfferCatalog/Art/Parallel/T_UI_Honeycomb}
  n21 = ScalarParameter() {defaultValue=0.3476, group=Smoke, parameterName=Smoke-Offset-X}
  n22 = ScalarParameter() {defaultValue=0.2238, group=Smoke, parameterName=Smoke-Offset-Y}
  n23 = Multiply(
   A=@n20.R
   B=MaterialFunctionCall(
    UVs=@n1
    CenterPosition=AppendVector(A=@n21, B=@n22)
    Radius=ScalarParameter() {defaultValue=0.35, group=Smoke, parameterName=Radius-Smoke, sortPriority=1}
    Density=ScalarParameter() {defaultValue=1.5, group=Smoke, parameterName=Density-Smoke, sortPriority=1}
   ) {materialFunction=&RadialGradientExponential}.RadialGradientExponential
  )
  n24 = VectorParameter() {parameterName=RGB2, defaultValue=<rgba 0.7157 0.0176 0.0044 1>}
  n25 = Multiply(
   A=VectorParameter() {parameterName=Color Opacity, defaultValue=<rgba 0.3542 0.3542 0.3542 5.6892>}.RGB
   B=VectorParameter() {parameterName=RGB1, defaultValue=<rgba 0.4391 0.1321 0.956 1>}.RGB
  )
  n26 = Multiply(
   A=Multiply(A=@n23, B=LinearInterpolate(A=@n24.RGB, B=@n25, Alpha=@n23))
   B=ScalarParameter() {defaultValue=1, group=Smoke, parameterName=Smoke-Intensity}
  )
  n27 = NamedRerouteUsage()
  n28 = NamedRerouteUsage()
  n29 = NamedRerouteUsage()
  n30 = NamedRerouteUsage()
  n31 = MaterialFunctionCall(Float2=MaterialFunctionCall() {materialFunction=&GetUserInterfaceUV}.Pixel Size) {materialFunction=&BreakOutFloat2Components}
  n32 = NamedRerouteDeclaration(None=@n31.G) {name=PixelSizeY, nodeColor=<rgba 0.4531 1 0 1>}
  n33 = NamedRerouteDeclaration(None=@n31.R) {name=PixelSizeX, nodeColor=<rgba 0 1 1 1>}
  n34 = AppendVector(A=Multiply(A=Divide(A=@n29, B=@n30), B=Divide(A=@n32, B=@n33)), B=Constant() {r=1})
  n35 = Divide(A=@n30, B=@n29)
  n36 = Divide(A=@n33, B=@n32)
  n37 = VectorParameter() {parameterName=UpperLeftIconSize, channelNames {a="", b="", g=Height, r=Width}, defaultValue=<rgba 100 100 0 1>}
  n38 = Divide(
   A=AppendVector(
    A=NamedRerouteDeclaration(None=@n37.Width) {name=UpperIconSizeX, nodeColor=<rgba 1 0.1406 0 1>}
    B=NamedRerouteDeclaration(None=@n37.Height) {name=UpperIconSizeY, nodeColor=<rgba 0 1 0.3438 1>}
   )
   B=MaterialFunctionCall() {materialFunction=&GetUserInterfaceUV}.Pixel Size
  )
  n39 = If(
   A=NamedRerouteUsage()
   B=NamedRerouteUsage()
   "A > B"=AppendVector(A=Constant() {r=1}, B=Multiply(A=@n35, B=@n36))
   "A == B"=@n38
   "A < B"=@n38
  )
  n40 = VectorParameter() {parameterName=UpperLeftPosition, channelNames {a="", b="", g=Y, r=X}, defaultValue=<rgba 0 0 1 1>}
  n41 = TextureSample(UVs=MaterialFunctionCall!(Texture Scale=If(A=@n27, B=@n28, "A > B"=@n34, "A == B"=@n39, "A < B"=@n39), UVs, Offset XY Texturesize Z=@n40) {materialFunction=/Engine/Functions/Engine_MaterialFunctions01/Texturing/ScaleUVsAroundPoint}.UVs) {samplerSource=SSM_Clamp_WorldGroupSettings, texture=/OfferCatalog/Art/Parallel/T_UI_Parallel_Corner2}
  n42 = VectorParameter() {parameterName=LowerRightPosition, channelNames {a="", b="", g=Y, r=X}, defaultValue=<rgba 1 1 1 1>}
  n43 = NamedRerouteUsage()
  n44 = NamedRerouteUsage()
  n45 = NamedRerouteUsage()
  n46 = NamedRerouteUsage()
  n47 = NamedRerouteUsage()
  n48 = NamedRerouteUsage()
  n49 = AppendVector(A=Multiply(A=Divide(A=@n45, B=@n46), B=Divide(A=@n47, B=@n48)), B=Constant() {r=1})
  n50 = Divide(A=@n46, B=@n45)
  n51 = Divide(A=@n48, B=@n47)
  n52 = VectorParameter() {parameterName=LowerRightIconSize, channelNames {a="", b="", g=Height, r=Width}, defaultValue=<rgba 100 100 0 1>}
  n53 = Divide(
   A=AppendVector(
    A=NamedRerouteDeclaration(None=@n52.Width) {name=LowerIconSizeX, nodeColor=<rgba 1 0.1406 0 1>}
    B=NamedRerouteDeclaration(None=@n52.Height) {name=LowerIconSizeY, nodeColor=<rgba 0 1 0.3438 1>}
   )
   B=MaterialFunctionCall() {materialFunction=&GetUserInterfaceUV}.Pixel Size
  )
  n54 = If(
   A=NamedRerouteUsage()
   B=NamedRerouteUsage()
   "A > B"=AppendVector(A=Constant() {r=1}, B=Multiply(A=@n50, B=@n51))
   "A == B"=@n53
   "A < B"=@n53
  )
  n55 = TextureSample(UVs=MaterialFunctionCall(Texture Scale=If(A=@n43, B=@n44, "A > B"=@n49, "A == B"=@n54, "A < B"=@n54), Offset XY Texturesize Z=@n42) {materialFunction=&ScaleUVsAroundPoint}.UVs) {samplerSource=SSM_Clamp_WorldGroupSettings, texture=/OfferCatalog/Art/Parallel/T_UI_Parallel_Corner1}
  LinearInterpolate(A=LinearInterpolate(A=Add(A=@n8.RGB, B=@n26), B=@n41.RGB, Alpha=@n41.RGB), B=@n55.RGB, Alpha=@n55.RGB)
  Multiply(A=@n15.Normalized UV, B=AppendVector(A=@n17, B=Constant() {r=1}))
  OneMinus(None=@n20.G)
  n56 = TextureCoordinate()
  n57 = ScalarParameter() {defaultValue=1, parameterName=Localized UV}
  n58 = Time()
  n59 = ScalarParameter() {defaultValue=0.15, parameterName=Time Factor}
  TextureSampleParameter2D!(UVs=Panner(Coordinate=Multiply(A=@n56, B=@n57), Time=Multiply(A=@n58, B=@n59)) {speedX=0.15, speedY=0.15}, Level, Apply View MipBias) {group=Smoke, mipValueMode=TMVM_MipLevel, parameterName=texture, texture=&T_UI_Honeycomb}
 ]
}
